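# Translucent subsurface particle material for snowball effects on the Nutcracker pickaxe, usable on regular and Niagara sprites.
# Particle color feeds an emissive adjustment function, while particle alpha is masked by a smoke texture or, via static switch, a SubUV snowflake sheet.
# Optional depth fade softens that opacity, bypassed on mobile.
# A default-on world position offset pulls particles onto a randomized-radius sphere around the object origin, weakened near the center.
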Material /BRCosmetics/Weapons/FORT_Melee/Pickaxe_NutCracker/FX/M_SnowBallParMat
graph {
 properties {
  bUsedWithNiagaraSprites=true
  bUsedWithParticleSprites=true
  blendMode=BLEND_Translucent
  pixelDepthOffsetMode=PDOM_Legacy
  shadingModel=MSM_Subsurface
  usedShadingModels=Subsurface
 }
 nodes [
  Constant3Vector() {constant=<rgba 0.9115 0.9608 1 0>}
  n1 = ParticleColor()
  MaterialFunctionCall(Emissive=@n1.RGB) {materialFunction=&MF_FX_EmissiveAdjustments}
  n2 = MaterialFunctionCall!(
   Texture=TextureObject() {texture=/Game/Effects/Fort_Effects/Textures/Snowflake/T_Snowflakes_F}
   UVs
   SubImages=AppendVector(A=ScalarParameter() {defaultValue=2, parameterName=SubUV_U}, B=ScalarParameter() {defaultValue=2, parameterName=SubUV_V})
   Frame=ScalarParameter() {parameterName=SubUV_Frame}
  ) {materialFunction=/Engine/Functions/Engine_MaterialFunctions03/Texturing/SubUV_Function}
  n3 = StaticSwitchParameter(
   True=Multiply(A=ComponentMask(None=@n2.RGB) {r=true}, B=@n1.A)
   False=Multiply(A=@n1.A, B=TextureSampleParameter2D() {parameterName=Snow_Texture, texture=/Game/Effects/Fort_Effects/Textures/Husk/T_Husk_Smoke}.R)
  ) {parameterName=Use SubUV Texture}
  n4 = DepthFade(Opacity=@n3, FadeDistance=ScalarParameter() {defaultValue=100, parameterName=Depth})
  StaticSwitchParameter(True=ShadingPathSwitch(Default=@n4, Deferred=@n4, Mobile=@n3) {desc=Migrated from FeatureLevelSwitch}, False=@n3) {parameterName=Use Depth Fade}
  n5 = Subtract(A=ParticlePositionWS(), B=ObjectPositionWS())
  n6 = Multiply(A=@n5) {constB=-1}
  n7 = ScalarParameter() {defaultValue=15, parameterName=Radius}
  n8 = ParticleRandom()
  n9 = Multiply(A=@n8, B=@n8)
  n10 = LinearInterpolate(A=@n7, B=Multiply(A=@n7) {constB=0.75}, Alpha=Multiply(A=@n9, B=@n9)) {constA=48, constB=30}
  n11 = Multiply(A=Normalize(VectorInput=@n5), B=@n10)
  n12 = Divide(A=DotProduct(A=@n5, B=@n5), B=@n10)
  StaticSwitchParameter(True=Multiply(A=Add(A=@n6, B=@n11), B=Saturate(None=@n12)), False=Constant()) {defaultValue=true, parameterName=Use Sphere Wpo}
 ]
}
