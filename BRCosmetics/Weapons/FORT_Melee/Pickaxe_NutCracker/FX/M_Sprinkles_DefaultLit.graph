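# Translucent lit sprite material for sprinkle particles in the Nutcracker pickaxe effects, with volumetric per-vertex directional lighting.
# Color comes from a per-particle random hue shift with desaturation or from particle color, feeding emissive, alongside a sprinkle normal map with adjustable strength.
# A circular UV mask times particle alpha gives opacity, with optional depth fade off mobile.
# Switchable world position offset, zero on mobile, pulls particles toward a randomized-radius sphere around the object position.
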Material /BRCosmetics/Weapons/FORT_Melee/Pickaxe_NutCracker/FX/M_Sprinkles_DefaultLit
graph {
 properties {
  bUsedWithNiagaraSprites=true
  bUsedWithParticleSprites=true
  blendMode=BLEND_Translucent
  pixelDepthOffsetMode=PDOM_Legacy
  translucencyLightingMode=TLM_VolumetricPerVertexDirectional
  usedShadingModels=Default Lit
 }
 nodes [
  MaterialFunctionCall(
   Normal=TextureSampleParameter2D() {parameterName=NormalMap, samplerType=SAMPLERTYPE_Normal, texture=/Game/Effects/Fort_Effects/Textures/T_Sprinkles_N}.RGB
   Flatness=Multiply(A=ScalarParameter() {parameterName=NormalStrength}) {constB=-1}
  ) {materialFunction=&FlattenNormal}
  n1 = MaterialFunctionCall(Hue Shift Percentage=ParticleRandom(), Texture=Constant3Vector() {constant=<rgba 1 0 0 0>}) {materialFunction=&HueShift}
  n2 = Desaturation(None=@n1.Result, Fraction=ScalarParameter() {parameterName=Desaturation})
  n3 = ParticleColor()
  MaterialFunctionCall(Emissive=Multiply(
    A=StaticSwitchParameter(True=@n2, False=@n3.RGB) {defaultValue=true, parameterName=UseRandomColor}
    B=ScalarParameter() {parameterName=Emissive}
   )) {materialFunction=&MF_FX_EmissiveAdjustments}
  Multiply(A=@n1.Result, B=@n3.RGB)
  ReflectionVectorWS()
  n4 = Multiply(A=SphereMask(A=TextureCoordinate(), B=Constant() {r=0.5}) {attenuationRadius=0.35}, B=@n3.A)
  n5 = DepthFade(Opacity=@n4, FadeDistance=ScalarParameter() {defaultValue=100, parameterName=Depth})
  StaticSwitchParameter(True=ShadingPathSwitch(Default=@n5, Deferred=@n5, Mobile=@n4) {desc=Migrated from FeatureLevelSwitch}, False=@n4) {defaultValue=true, parameterName=Use Depth Fade}
  n6 = Subtract(A=ParticlePositionWS(), B=ObjectPositionWS())
  n7 = StaticSwitchParameter(True=ParticleColor().R, False=ScalarParameter() {defaultValue=15, parameterName=Radius}) {parameterName=UseParticleDP}
  n8 = ParticleRandom()
  n9 = Multiply(A=@n8, B=@n8)
  n10 = LinearInterpolate(A=@n7, B=Multiply(A=@n7) {constB=0.75}, Alpha=Multiply(A=@n9, B=@n9)) {constA=48, constB=30}
  n11 = Add(A=Multiply(A=@n6) {constB=-1}, B=Multiply(A=Normalize(VectorInput=@n6), B=@n10))
  n12 = Saturate(None=Divide(A=DotProduct(A=@n6, B=@n6), B=@n10))
  StaticSwitchParameter(True=ShadingPathSwitch(Default=Multiply(A=@n11, B=@n12), Mobile=Constant()) {desc=Migrated from FeatureLevelSwitch}, False=Constant()) {defaultValue=true, parameterName=Use Sphere Wpo}
 ]
}
