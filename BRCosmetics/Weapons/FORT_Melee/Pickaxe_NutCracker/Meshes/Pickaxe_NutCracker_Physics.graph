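# Physics asset for the NutCracker two-handed pickaxe, giving its snow globe, leaves, bread and tassel bones simulated secondary motion.
# A capsule on the handle anchors the globe base sphere, which in turn carries the snow, two tightly limited leaf capsules and fourteen three-capsule tassel chains.
# The bread hangs off the base through an intermediate rotate body with free angular motion on both joints, while the snow spins with one swing axis locked.
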PhysicsAsset /BRCosmetics/Weapons/FORT_Melee/Pickaxe_NutCracker/Meshes/Pickaxe_NutCracker_Physics
bodies {
 Melee_TwoHanded_Handle [{shapeType=Capsule, length=94.9993, radius=1.9115, center=<0.8195 -0.8313 9.5604>, rotation=<rot -0.2699 -89.9643 -0.1242>}]
 dyn_bread [{shapeType=Sphere, radius=3.1258, center=<0.2728 -0.3786 -0.2795>}]
 dyn_breadROTATE [{shapeType=Sphere, radius=3.125}]
 dyn_leaf_01 [{shapeType=Capsule, length=6.1663, radius=2.0653, center=<4.2143 -0.0056 -0.269>, rotation=<rot 0 91.5934 -74.378>}]
 dyn_leaf_02 [{shapeType=Capsule, length=4.8133, radius=1.8761, center=<4.7036 0.4041 -0.0504>, rotation=<rot 0 -92.0264 -76.993>}]
 dyn_snow [{shapeType=Sphere, radius=10.5}]
 dyn_snowGlobe_base [{shapeType=Sphere, radius=7.505}]
 tassleA_01 [{shapeType=Capsule, length=1.9411, radius=0.8623, center=<1.1793 0 0>, rotation=<rot 0 90 -90>}]
 tassleA_02 [{shapeType=Capsule, length=2.4411, radius=0.8623, center=<1.1793 0 0>, rotation=<rot 0 90 -90>}]
 tassleA_03 [{shapeType=Capsule, length=2.5046, radius=0.8824, center=<2.4652 -0.0008 -0.0005>, rotation=<rot 0 -90.4818 -89.0357>}]
 tassleB_01 [{shapeType=Capsule, length=1.9411, radius=0.8656, center=<1.1793 0 0>, rotation=<rot 0 -90 -90>}]
 tassleB_02 [{shapeType=Capsule, length=2.4411, radius=0.8656, center=<1.1793 0 0>, rotation=<rot 0 -90 -90>}]
 tassleB_03 [{shapeType=Capsule, length=1.3289, radius=0.8919, center=<1.2831 -0.0021 -0.0019>, rotation=<rot 0 -91.2683 -87.9493>}]
 tassleC_01 [{shapeType=Capsule, length=1.9411, radius=0.8625, center=<1.1793 0 0>, rotation=<rot 0 90 -90>}]
 tassleC_02 [{shapeType=Capsule, length=2.4411, radius=0.8625, center=<1.1793 0 0>, rotation=<rot 0 89.9999 -90>}]
 tassleC_03 [{shapeType=Capsule, length=2.5046, radius=0.8827, center=<2.4652 -0.0008 -0.0005>, rotation=<rot 0 -90.484 -89.0368>}]
 tassleD_01 [{shapeType=Capsule, length=1.9411, radius=0.8653, center=<1.1793 0 0>, rotation=<rot 0 -90 -90>}]
 tassleD_02 [{shapeType=Capsule, length=2.4411, radius=0.8653, center=<1.1793 0 0>, rotation=<rot 0 90 -90>}]
 tassleD_03 [{shapeType=Capsule, length=2.5046, radius=0.8877, center=<2.4652 -0.0009 -0.0002>, rotation=<rot 0 -90.5381 -89.066>}]
 tassleE_01 [{shapeType=Capsule, length=1.9411, radius=0.8632, center=<1.1793 0 0>, rotation=<rot 0 -90 -90>}]
 tassleE_02 [{shapeType=Capsule, length=2.4411, radius=0.8632, center=<1.1793 0 0>, rotation=<rot 0 90 -90>}]
 tassleE_03 [{shapeType=Capsule, length=2.5046, radius=0.8895, center=<2.4652 -0.0011 -0.0003>, rotation=<rot 0 -90.6309 -89.1259>}]
 tassleF_01 [{shapeType=Capsule, length=1.9411, radius=0.8651, center=<1.1793 0 0>, rotation=<rot 0 90 -90>}]
 tassleF_02 [{shapeType=Capsule, length=2.4411, radius=0.8651, center=<1.1793 0 0>, rotation=<rot 0 90 -90>}]
 tassleF_03 [{shapeType=Capsule, length=2.5046, radius=0.8899, center=<2.4652 -0.001 -0.0002>, rotation=<rot 0 -90.5939 -89.1004>}]
 tassleG_01 [{shapeType=Capsule, length=1.9411, radius=0.864, center=<1.1793 0 0>, rotation=<rot 0 -90 -90>}]
 tassleG_02 [{shapeType=Capsule, length=2.4411, radius=0.864, center=<1.1793 0 0>, rotation=<rot 0 -90 -90>}]
 tassleG_03 [{shapeType=Capsule, length=1.3289, radius=0.8881, center=<1.2831 -0.0019 -0.0022>, rotation=<rot 0 -91.1579 -87.8851>}]
 tassleH_01 [{shapeType=Capsule, length=1.9411, radius=0.8649, center=<1.1793 0 0>, rotation=<rot 0 -90 -90>}]
 tassleH_02 [{shapeType=Capsule, length=2.4411, radius=0.8649, center=<1.1793 0 0>, rotation=<rot 0 -90 -90>}]
 tassleH_03 [{shapeType=Capsule, length=2.5046, radius=0.8869, center=<2.4652 -0.0009 -0.0003>, rotation=<rot 0 -90.5267 -89.0595>}]
 tassleJ_01 [{shapeType=Capsule, length=1.9411, radius=0.8655, center=<1.1793 0 0>, rotation=<rot 0 90 -90>}]
 tassleJ_02 [{shapeType=Capsule, length=2.4411, radius=0.8655, center=<1.1793 0 0>, rotation=<rot 0 -90 -90>}]
 tassleJ_03 [{shapeType=Capsule, length=2.5046, radius=0.8895, center=<2.4652 -0.001 -0.0001>, rotation=<rot 0 -90.5756 -89.0886>}]
 tassleK_01 [{shapeType=Capsule, length=1.9411, radius=0.8578, center=<1.1793 0 0>, rotation=<rot 0 90 -90>}]
 tassleK_02 [{shapeType=Capsule, length=2.4411, radius=0.8578, center=<1.1793 0 0>, rotation=<rot 0 -90 -90>}]
 tassleK_03 [{shapeType=Capsule, length=1.3289, radius=0.8786, center=<1.2831 -0.0017 -0.0026>, rotation=<rot 0 -90.9995 -87.8058>}]
 tassleL_01 [{shapeType=Capsule, length=1.9411, radius=0.8646, center=<1.1793 0 0>, rotation=<rot 0 -90 -90>}]
 tassleL_02 [{shapeType=Capsule, length=2.4411, radius=0.8646, center=<1.1793 0 0>, rotation=<rot 0 90 -90>}]
 tassleL_03 [{shapeType=Capsule, length=2.5046, radius=0.8899, center=<2.4652 -0.001 -0.0002>, rotation=<rot 0 -90.6079 -89.1098>}]
 tassleM_01 [{shapeType=Capsule, length=1.9411, radius=0.8649, center=<1.1793 0 0>, rotation=<rot 0 90 -90>}]
 tassleM_02 [{shapeType=Capsule, length=2.4411, radius=0.8649, center=<1.1793 0 0>, rotation=<rot 0 90 -90>}]
 tassleM_03 [{shapeType=Capsule, length=2.5046, radius=0.8869, center=<2.4652 -0.0009 -0.0003>, rotation=<rot 0 -90.5267 -89.0595>}]
 tassleN_01 [{shapeType=Capsule, length=1.9411, radius=0.8649, center=<1.1793 0 0>, rotation=<rot 0 -90 -90>}]
 tassleN_02 [{shapeType=Capsule, length=2.4411, radius=0.8649, center=<1.1793 0 0>, rotation=<rot 0 -90 -90>}]
 tassleN_03 [{shapeType=Capsule, length=2.5046, radius=0.8869, center=<2.4652 -0.0009 -0.0003>, rotation=<rot 0 -90.5267 -89.0595>}]
 tassleP_01 [{shapeType=Capsule, length=1.9411, radius=0.8649, center=<1.1793 0 0>, rotation=<rot 0 -90 -89.9999>}]
 tassleP_02 [{shapeType=Capsule, length=2.4411, radius=0.8649, center=<1.1793 0 0>, rotation=<rot 0 90 -90>}]
 tassleP_03 [{shapeType=Capsule, length=2.5046, radius=0.8869, center=<2.4652 -0.0009 -0.0003>, rotation=<rot 0 -90.5267 -89.0595>}]
}
constraints [
 {bone1Name=dyn_snowGlobe_base, bone2Name=Melee_TwoHanded_Handle, swing1LimitDegrees=45, swing2LimitDegrees=45, twistLimitDegrees=45}
 {
  bone1Name=dyn_snow
  bone2Name=dyn_snowGlobe_base
  swing1LimitDegrees=45
  swing1Motion=Locked
  swing2LimitDegrees=45
  swing2Motion=Free
  twistLimitDegrees=45
  twistMotion=Free
 }
 {bone1Name=dyn_leaf_01, bone2Name=dyn_snowGlobe_base, swing1LimitDegrees=5, swing2LimitDegrees=5, twistLimitDegrees=45, twistMotion=Locked}
 {bone1Name=dyn_leaf_02, bone2Name=dyn_snowGlobe_base, swing1LimitDegrees=5, swing2LimitDegrees=5, twistLimitDegrees=45, twistMotion=Locked}
 {
  bone1Name=dyn_breadROTATE
  bone2Name=dyn_snowGlobe_base
  swing1LimitDegrees=45
  swing1Motion=Free
  swing2LimitDegrees=45
  swing2Motion=Free
  twistLimitDegrees=45
  twistMotion=Free
 }
 {
  bone1Name=dyn_bread
  bone2Name=dyn_breadROTATE
  swing1LimitDegrees=45
  swing1Motion=Free
  swing2LimitDegrees=45
  swing2Motion=Free
  twistLimitDegrees=45
  twistMotion=Free
 }
 {bone1Name=tassleA_01, bone2Name=dyn_snowGlobe_base, swing1LimitDegrees=35, swing2LimitDegrees=35, twistLimitDegrees=10}
 {bone1Name=tassleA_02, bone2Name=tassleA_01, swing1LimitDegrees=20, swing2LimitDegrees=20, twistLimitDegrees=5}
 {bone1Name=tassleA_03, bone2Name=tassleA_02, swing1LimitDegrees=30, swing2LimitDegrees=30, twistLimitDegrees=5}
 {bone1Name=tassleB_01, bone2Name=dyn_snowGlobe_base, swing1LimitDegrees=35, swing2LimitDegrees=35, twistLimitDegrees=10}
 {bone1Name=tassleB_02, bone2Name=tassleB_01, swing1LimitDegrees=20, swing2LimitDegrees=20, twistLimitDegrees=5}
 {bone1Name=tassleB_03, bone2Name=tassleB_02, swing1LimitDegrees=30, swing2LimitDegrees=30, twistLimitDegrees=5}
 {bone1Name=tassleC_01, bone2Name=dyn_snowGlobe_base, swing1LimitDegrees=35, swing2LimitDegrees=35, twistLimitDegrees=10}
 {bone1Name=tassleC_02, bone2Name=tassleC_01, swing1LimitDegrees=20, swing2LimitDegrees=20, twistLimitDegrees=5}
 {bone1Name=tassleC_03, bone2Name=tassleC_02, swing1LimitDegrees=30, swing2LimitDegrees=30, twistLimitDegrees=5}
 {bone1Name=tassleD_01, bone2Name=dyn_snowGlobe_base, swing1LimitDegrees=35, swing2LimitDegrees=35, twistLimitDegrees=10}
 {bone1Name=tassleD_02, bone2Name=tassleD_01, swing1LimitDegrees=20, swing2LimitDegrees=20, twistLimitDegrees=5}
 {bone1Name=tassleD_03, bone2Name=tassleD_02, swing1LimitDegrees=30, swing2LimitDegrees=30, twistLimitDegrees=5}
 {bone1Name=tassleE_01, bone2Name=dyn_snowGlobe_base, swing1LimitDegrees=35, swing2LimitDegrees=35, twistLimitDegrees=10}
 {bone1Name=tassleE_02, bone2Name=tassleE_01, swing1LimitDegrees=20, swing2LimitDegrees=20, twistLimitDegrees=5}
 {bone1Name=tassleE_03, bone2Name=tassleE_02, swing1LimitDegrees=30, swing2LimitDegrees=30, twistLimitDegrees=5}
 {bone1Name=tassleF_01, bone2Name=dyn_snowGlobe_base, swing1LimitDegrees=35, swing2LimitDegrees=35, twistLimitDegrees=10}
 {bone1Name=tassleF_02, bone2Name=tassleF_01, swing1LimitDegrees=20, swing2LimitDegrees=20, twistLimitDegrees=5}
 {bone1Name=tassleF_03, bone2Name=tassleF_02, swing1LimitDegrees=30, swing2LimitDegrees=30, twistLimitDegrees=5}
 {bone1Name=tassleG_01, bone2Name=dyn_snowGlobe_base, swing1LimitDegrees=35, swing2LimitDegrees=35, twistLimitDegrees=10}
 {bone1Name=tassleG_02, bone2Name=tassleG_01, swing1LimitDegrees=20, swing2LimitDegrees=20, twistLimitDegrees=5}
 {bone1Name=tassleG_03, bone2Name=tassleG_02, swing1LimitDegrees=30, swing2LimitDegrees=30, twistLimitDegrees=5}
 {bone1Name=tassleH_01, bone2Name=dyn_snowGlobe_base, swing1LimitDegrees=35, swing2LimitDegrees=35, twistLimitDegrees=10}
 {bone1Name=tassleH_02, bone2Name=tassleH_01, swing1LimitDegrees=20, swing2LimitDegrees=20, twistLimitDegrees=5}
 {bone1Name=tassleH_03, bone2Name=tassleH_02, swing1LimitDegrees=30, swing2LimitDegrees=30, twistLimitDegrees=5}
 {bone1Name=tassleJ_01, bone2Name=dyn_snowGlobe_base, swing1LimitDegrees=35, swing2LimitDegrees=35, twistLimitDegrees=10}
 {bone1Name=tassleJ_02, bone2Name=tassleJ_01, swing1LimitDegrees=20, swing2LimitDegrees=20, twistLimitDegrees=5}
 {bone1Name=tassleJ_03, bone2Name=tassleJ_02, swing1LimitDegrees=30, swing2LimitDegrees=30, twistLimitDegrees=5}
 {bone1Name=tassleK_01, bone2Name=dyn_snowGlobe_base, swing1LimitDegrees=35, swing2LimitDegrees=35, twistLimitDegrees=10}
 {bone1Name=tassleK_02, bone2Name=tassleK_01, swing1LimitDegrees=20, swing2LimitDegrees=20, twistLimitDegrees=5}
 {bone1Name=tassleK_03, bone2Name=tassleK_02, swing1LimitDegrees=30, swing2LimitDegrees=30, twistLimitDegrees=5}
 {bone1Name=tassleL_01, bone2Name=dyn_snowGlobe_base, swing1LimitDegrees=35, swing2LimitDegrees=35, twistLimitDegrees=10}
 {bone1Name=tassleL_02, bone2Name=tassleL_01, swing1LimitDegrees=20, swing2LimitDegrees=20, twistLimitDegrees=5}
 {bone1Name=tassleL_03, bone2Name=tassleL_02, swing1LimitDegrees=30, swing2LimitDegrees=30, twistLimitDegrees=5}
 {bone1Name=tassleM_01, bone2Name=dyn_snowGlobe_base, swing1LimitDegrees=35, swing2LimitDegrees=35, twistLimitDegrees=10}
 {bone1Name=tassleM_02, bone2Name=tassleM_01, swing1LimitDegrees=20, swing2LimitDegrees=20, twistLimitDegrees=5}
 {bone1Name=tassleM_03, bone2Name=tassleM_02, swing1LimitDegrees=30, swing2LimitDegrees=30, twistLimitDegrees=5}
 {bone1Name=tassleN_01, bone2Name=dyn_snowGlobe_base, swing1LimitDegrees=35, swing2LimitDegrees=35, twistLimitDegrees=10}
 {bone1Name=tassleN_02, bone2Name=tassleN_01, swing1LimitDegrees=20, swing2LimitDegrees=20, twistLimitDegrees=5}
 {bone1Name=tassleN_03, bone2Name=tassleN_02, swing1LimitDegrees=30, swing2LimitDegrees=30, twistLimitDegrees=5}
 {bone1Name=tassleP_01, bone2Name=dyn_snowGlobe_base, swing1LimitDegrees=35, swing2LimitDegrees=35, twistLimitDegrees=10}
 {bone1Name=tassleP_02, bone2Name=tassleP_01, swing1LimitDegrees=20, swing2LimitDegrees=20, twistLimitDegrees=5}
 {bone1Name=tassleP_03, bone2Name=tassleP_02, swing1LimitDegrees=30, swing2LimitDegrees=30, twistLimitDegrees=5}
]
