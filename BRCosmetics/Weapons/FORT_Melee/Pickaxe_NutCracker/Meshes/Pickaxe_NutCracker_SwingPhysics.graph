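# Physics asset for the NutCracker pickaxe, a two-handed melee cosmetic, setting up a small swing-physics chain on its mesh.
# Three sphere bodies sit on the handle, dyn_swingMeasureBASE and dyn_swingMeasure bones.
# The BASE body rotates freely on all axes relative to the handle, while dyn_swingMeasure is angularly locked to BASE, so the pair swings as one pendulum-like unit.
# The bone names suggest the pair serves to measure swing motion.
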PhysicsAsset /BRCosmetics/Weapons/FORT_Melee/Pickaxe_NutCracker/Meshes/Pickaxe_NutCracker_SwingPhysics
bodies {
 Melee_TwoHanded_Handle [{shapeType=Sphere, radius=11.0078, center=<0.3129 -1.1676 31.1383>}]
 dyn_swingMeasure [{shapeType=Sphere, radius=2.255}]
 dyn_swingMeasureBASE [{shapeType=Sphere, radius=3.875}]
}
constraints [
 {
  bone1Name=dyn_swingMeasure
  bone2Name=dyn_swingMeasureBASE
  swing1LimitDegrees=45
  swing1Motion=Locked
  swing2LimitDegrees=45
  swing2Motion=Locked
  twistLimitDegrees=45
  twistMotion=Locked
 }
 {
  bone1Name=dyn_swingMeasureBASE
  bone2Name=Melee_TwoHanded_Handle
  swing1LimitDegrees=45
  swing1Motion=Free
  swing2LimitDegrees=45
  swing2Motion=Free
  twistLimitDegrees=45
  twistMotion=Free
 }
]
