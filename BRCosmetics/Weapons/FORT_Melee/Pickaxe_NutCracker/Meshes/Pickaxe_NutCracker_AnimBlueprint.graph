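# Animation Blueprint for the NutCracker pickaxe, adding secondary motion and spinning parts to the mesh.
# AnimGraph chains RigidBody physics, a trail on dyn_swingMeasure, then bone modifiers for center travel, center spin and bread tumble.
# EventGraph eases SwingDelta by SwingBlend, scaling travel and a wrapped spin angle that advances per update, independent of delta time.
# The socket-distance test setting SwingBlend and a second swing-physics RigidBody are unconnected.
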AnimBlueprint /BRCosmetics/Weapons/FORT_Melee/Pickaxe_NutCracker/Meshes/Pickaxe_NutCracker_AnimBlueprint
target_skeleton=&/Pickaxe_NutCracker_Skeleton
graphs [
 AnimGraph {nodes [
   RigidBody() {
    bOverrideWorldGravity=true
    overridePhysicsAsset=&/Pickaxe_NutCracker_SwingPhysics
    alphaBoolBlend {blendOption=Linear}
    componentAppliedLinearAccClamp=<5000 5000 5000>
    componentLinearAccScale=<0.4 0.4 0.4>
    componentLinearVelScale=<0.3 0.3 0.3>
   }
   AnimGraphNode_SubInput_0 = Input Pose()
   n1 = RigidBody(ComponentPose=Local To Component(LocalPose=@AnimGraphNode_SubInput_0)) {
    bOverrideWorldGravity=true
    overridePhysicsAsset=&/Pickaxe_NutCracker_Physics
    alphaBoolBlend {blendOption=Linear}
    componentLinearAccScale=<0.1 0.1 0.1>
    componentLinearVelScale=<0.1 0.1 0.1>
    overrideWorldGravity {z=-980}
   }
   n2 = Trail controller - Bone: dyn_swingMeasure(ComponentPose=@n1, Alpha=1) {
    rotationLimits [{limitMax=<180 180 180>}, {limitMax=<180 180 180>}]
    rotationOffsets [<0 0 0>, <0 0 0>]
    trailBone {boneName=dyn_swingMeasure}
    bLimitStretch=~
   }
   n3 = "Transform (Modify) Bone - Bone: dyn_centerTRAVEL"(ComponentPose=@n2, Alpha=Get SwingDelta() -> (SwingDelta:double), Translation="5,0,0", Rotation="0,0,0", Scale="1,1,1") {
    translationMode=BMM_Additive
    translationSpace=BCS_ParentBoneSpace
    alphaBoolBlend {blendOption=Linear}
    alphaScaleBiasClamp {bMapRange=true, clampMin=0.25, interpSpeedDecreasing=1, interpSpeedIncreasing=1, inRange {max=21, min=1}, outRange {min=0.35}}
    boneToModify {boneName=dyn_centerTRAVEL}
    translation {x=5}
   }
   n4 = "Transform (Modify) Bone - Bone: dyn_centerMainROTATE"(ComponentPose=@n3, Alpha=1, Translation="0,0,0", Rotation=Get SpinConst() -> (SpinConst:Rotator), Scale="1,1,1") {
    rotationMode=BMM_Replace
    rotationSpace=BCS_BoneSpace
    translationMode=BMM_Additive
    translationSpace=BCS_ParentBoneSpace
    alphaBoolBlend {blendOption=Linear}
    alphaScaleBiasClamp {interpSpeedDecreasing=1, interpSpeedIncreasing=1}
    boneToModify {boneName=dyn_centerMainROTATE}
   }
   n5 = "Transform (Modify) Bone - Bone: dyn_breadROTATE"(ComponentPose=@n4, Alpha=1, Translation="0,0,0", Rotation=Get GingerTumble() -> (GingerTumble:Rotator), Scale="1,1,1") {
    rotationMode=BMM_Additive
    rotationSpace=BCS_BoneSpace
    translationMode=BMM_Additive
    translationSpace=BCS_ParentBoneSpace
    alphaBoolBlend {blendOption=Linear}
    alphaScaleBiasClamp {interpSpeedDecreasing=1, interpSpeedIncreasing=1}
    boneToModify {boneName=dyn_breadROTATE}
   }
   Output Pose(Result=Component To Local(ComponentPose=@n5)) {layerGroup=None}
  ]}
 EventGraph {nodes [
   n1 = Event BlueprintUpdateAnimation()
   > n2 = Sequence()
   > n3 = Branch(Condition="float >= float"(A=Get SpinX2() -> (SpinX2:double), B=720) {is=title K2Node_CallFunction})
   > n4 = Set SpinX2(SpinX2:double=float - float(A=Get SpinX2(), B=720) {is=title K2Node_CallFunction}) -> (Output_Get:double)
   n5 = float + float(A=Get SpinX2(), B=float * float(A=Get SwingDelta(), B=Get SpinPerSecondFast() -> (SpinPerSecondFast:double))) {is=title K2Node_CommutativeAssociativeBinaryOperator}
   Set SpinConst(execute=[@n7.then, @n4.then], SpinConst:Rotator=MakeRotator(Yaw=@n5)) -> (Output_Get:Rotator)
   > Set GingerTumble(GingerTumble:Rotator=MakeRotator(Roll=@n5, Pitch=@n5)) -> (Output_Get:Rotator)
   n6 = GetPartSkeletalMeshComponent()
   > Set SwingBlend(SwingBlend:bool="float >= float"(
     A=Get Distance Between Two Sockets!(
      Component:Skeletal Mesh Component Object Reference=@n6
      SocketOrBoneNameA:Name=axeTop
      SocketSpaceA:ERelativeTransformSpace Enum=RTS_Actor
      SocketOrBoneNameB:Name=axeSwing
      SocketSpaceB:ERelativeTransformSpace Enum=RTS_Actor
      bRemapRange:bool=false
      InRangeMin:float=4
      InRangeMax:float=45
      OutRangeMin:float=0
      OutRangeMax:float=1
     ) -> (ReturnValue:float) {is=title K2Node_CallFunction}
     B=40
    )) -> (Output_Get:bool)
   n7 = Set SpinX2(execute=@n3.else, SpinX2=@n5)
   n8 = Branch(execute=@n2.then_1, Condition=Get SwingBlend() -> (SwingBlend:bool))
   > Set SwingDelta(SwingDelta:double=FInterpTo(Current=Get SwingDelta(), Target=21, DeltaTime=@n1.DeltaTimeX, InterpSpeed=1)) -> (Output_Get:double)
   Set SwingDelta(execute=@n8.else, SwingDelta=FInterpTo(Current=Get SwingDelta(), Target=1, DeltaTime=@n1.DeltaTimeX, InterpSpeed=1))
  ]}
]
variables [
 GingerTumble {type=Rotator}
 SpinConst {type=Rotator}
 SpinPerSecondFast {type=number, default=0.85}
 SpinX2 {type=number}
 SwingBlend {type=boolean}
 SwingDelta {type=number, default=1}
]
